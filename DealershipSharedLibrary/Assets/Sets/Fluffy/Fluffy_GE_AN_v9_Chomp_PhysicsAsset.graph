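# Physics asset for the Fluffy Chomp skeletal mesh, a face rig with mouth, lips, teeth, uvula and eyes.
# Each listed bone gets a single capsule body, with a large one on Rig_Root as the hub.
# Inner mouth, lip, eye, teeth and upper uvula bodies constrain directly to the root, while lip spring bodies hang off lips and the uvula forms a two-link chain.
# All constraints share identical swing and twist limit angles.
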PhysicsAsset /DealershipSharedLibrary/Assets/Sets/Fluffy/Fluffy_GE_AN_v9_Chomp_PhysicsAsset
bodies {
 MouthInner_001 [{shapeType=Capsule, length=229.9342, radius=193.4947, center=<2.2293 41.9386 -17.1566>, rotation=<rot 0 176.0794 -66.3398>}]
 MouthInner_002 [{shapeType=Capsule, length=228.6835, radius=226.755, center=<-11.4186 -5.0662 -58.9177>, rotation=<rot 0 42.8162 -77.4883>}]
 MouthInner_003 [{shapeType=Capsule, length=661.4374, radius=175.3009, center=<-54.1739 45.1636 54.9641>, rotation=<rot 0 -92.7874 -86.0331>}]
 MouthInner_004 [{shapeType=Capsule, length=589.2327, radius=202.1472, center=<-68.2 75.6791 -95.0535>, rotation=<rot 0 91.1348 -89.0425>}]
 MouthInner_005 [{shapeType=Capsule, length=272.2957, radius=188.3358, center=<66.1467 53.0006 -61.0073>, rotation=<rot 0 139.2821 -76.843>}]
 MouthInner_006 [{shapeType=Capsule, length=393.6366, radius=188.8182, center=<-13.4994 93.3612 -51.8085>, rotation=<rot 0 45.7475 -84.065>}]
 MouthInner_007 [{shapeType=Capsule, length=494.9896, radius=213.3277, center=<84.6058 94.9585 50.5962>, rotation=<rot 0 -95.5619 -85.0204>}]
 MouthInner_008 [{shapeType=Capsule, length=315.0461, radius=211.567, center=<60.8891 131.4243 -61.8123>, rotation=<rot 0 68.1255 -85.1021>}]
 Rig_Eye_L_Parent [{shapeType=Capsule, length=95.5974, radius=80.1956, center=<-1.8573 3.019 0.0529>, rotation=<rot 44.2768 -90 -89.2452>}]
 Rig_Eye_R [{shapeType=Capsule, length=95.5975, radius=80.1956, center=<-0.0471 2.8728 -8.0584>, rotation=<rot 44.277 -90 -90.7531>}]
 Rig_Lip_001 [{shapeType=Capsule, length=275.3362, radius=217.9658, center=<78.7761 5.784 90.7048>, rotation=<rot 0 -99.423 -84.4583>}]
 Rig_Lip_002 [{shapeType=Capsule, length=269.6247, radius=184.9483, center=<102.8016 1.4342 -59.7714>, rotation=<rot 0 -103.2352 -49.2181>}]
 Rig_Lip_003 [{shapeType=Capsule, length=199.6087, radius=124.3487, center=<-49.585 -2.8801 2.3277>, rotation=<rot 0 -80.8652 -73.8244>}]
 Rig_Lip_004 [{shapeType=Capsule, length=269.4414, radius=174.8866, center=<13.7213 7.0961 26.978>, rotation=<rot 0 -84.5699 -77.0884>}]
 Rig_Lip_005 [{shapeType=Capsule, length=263.4484, radius=188.8108, center=<-20.9426 -1.7469 33.6034>, rotation=<rot 0 -89.1648 -86.6233>}]
 Rig_Lip_006 [{shapeType=Capsule, length=275.5972, radius=162.9276, center=<-124.8251 7.3088 3.534>, rotation=<rot 0 91.3417 -89.6803>}]
 Rig_Lip_007 [{shapeType=Capsule, length=267.6597, radius=111.6391, center=<-26.439 -3.2274 -5.4829>, rotation=<rot 0 -95.6851 -76.2726>}]
 Rig_Lip_008 [{shapeType=Capsule, length=206.1087, radius=144.4924, center=<-37.3054 18.3797 57.8456>, rotation=<rot 0 95.8371 -78.0883>}]
 Rig_Lip_009 [{shapeType=Capsule, length=271.9012, radius=234.5909, center=<-68.3608 9.914 94.1168>, rotation=<rot 0 99.1155 -81.921>}]
 Rig_Lip_Spring_001 [{shapeType=Capsule, length=120.6615, radius=64.6871, center=<192.7001 -48.3897 -105.6415>, rotation=<rot 0 -88.9209 -71.8156>}]
 Rig_Lip_Spring_002 [{shapeType=Capsule, length=198.9048, radius=64.6362, center=<95.6931 -43.1495 -115.401>, rotation=<rot 0 90.477 -87.8936>}]
 Rig_Lip_Spring_003 [{shapeType=Capsule, length=98.4866, radius=40.6874, center=<241.5167 -40.8287 -129.5395>, rotation=<rot 0 -102.2152 -80.0262>}]
 Rig_Lip_Spring_004 [{shapeType=Capsule, length=180.772, radius=33.3612, center=<133.6048 -56.1402 -176.9168>, rotation=<rot 0 92.0809 -87.8107>}]
 Rig_Lip_Spring_005 [{shapeType=Capsule, length=141.5224, radius=52.9746, center=<-41.3715 -50.0507 -206.9485>, rotation=<rot 3.8047 -90 -44.6929>}]
 Rig_Lip_Spring_006 [{shapeType=Capsule, length=84.7591, radius=44.3745, center=<-77.3826 -55.5362 29.4013>, rotation=<rot 16.5203 -90 -15.5708>}]
 Rig_Lip_Spring_007 [{shapeType=Capsule, length=123.5059, radius=55.4012, center=<63.4455 -62.1576 -31.9465>, rotation=<rot -8.09 -90 3.4351>}]
 Rig_Lip_Spring_008 [{shapeType=Capsule, length=334.5702, radius=118.8971, center=<128.5441 -55.2818 -193.3097>, rotation=<rot 0 92.2034 -66.4366>}]
 Rig_Root [{shapeType=Capsule, length=910.9894, radius=465.1913, center=<-9.2929 278.8002 320.251>, rotation=<rot 0 -89.9033 -89.5921>}]
 Rig_Teeth_Bottom [{shapeType=Capsule, length=833.1732, radius=142.2125, center=<-38.5347 -38.254 -134.1435>, rotation=<rot 0 90.0124 -89.1126>}]
 Rig_Teeth_Top [{shapeType=Capsule, length=666.1823, radius=158.1036, center=<-33.7525 -46.4395 43.3448>, rotation=<rot 0 89.9752 -89.7126>}]
 Rig_Teeth_Uvula_001 [{shapeType=Capsule, length=47.2731, radius=43.0533, center=<-0.3305 -1.2368 -25.6483>, rotation=<rot 1.9765 -90 -90>}]
 Rig_Teeth_Uvula_003 [{shapeType=Capsule, length=44.1851, radius=44.1024, center=<-0.3305 0.2017 -29.2047>, rotation=<rot 0.2132 -90 0>}]
}
constraints [
 {bone1Name=Rig_Lip_009, bone2Name=Rig_Root, swing1LimitDegrees=45, swing2LimitDegrees=45, twistLimitDegrees=45}
 {bone1Name=Rig_Lip_008, bone2Name=Rig_Root, swing1LimitDegrees=45, swing2LimitDegrees=45, twistLimitDegrees=45}
 {bone1Name=Rig_Lip_Spring_007, bone2Name=Rig_Lip_008, swing1LimitDegrees=45, swing2LimitDegrees=45, twistLimitDegrees=45}
 {bone1Name=Rig_Lip_007, bone2Name=Rig_Root, swing1LimitDegrees=45, swing2LimitDegrees=45, twistLimitDegrees=45}
 {bone1Name=Rig_Lip_Spring_008, bone2Name=Rig_Lip_007, swing1LimitDegrees=45, swing2LimitDegrees=45, twistLimitDegrees=45}
 {bone1Name=Rig_Lip_006, bone2Name=Rig_Root, swing1LimitDegrees=45, swing2LimitDegrees=45, twistLimitDegrees=45}
 {bone1Name=Rig_Lip_Spring_001, bone2Name=Rig_Lip_006, swing1LimitDegrees=45, swing2LimitDegrees=45, twistLimitDegrees=45}
 {bone1Name=Rig_Lip_005, bone2Name=Rig_Root, swing1LimitDegrees=45, swing2LimitDegrees=45, twistLimitDegrees=45}
 {bone1Name=Rig_Lip_Spring_003, bone2Name=Rig_Lip_005, swing1LimitDegrees=45, swing2LimitDegrees=45, twistLimitDegrees=45}
 {bone1Name=Rig_Lip_Spring_002, bone2Name=Rig_Lip_005, swing1LimitDegrees=45, swing2LimitDegrees=45, twistLimitDegrees=45}
 {bone1Name=Rig_Lip_004, bone2Name=Rig_Root, swing1LimitDegrees=45, swing2LimitDegrees=45, twistLimitDegrees=45}
 {bone1Name=Rig_Lip_Spring_004, bone2Name=Rig_Lip_004, swing1LimitDegrees=45, swing2LimitDegrees=45, twistLimitDegrees=45}
 {bone1Name=Rig_Lip_002, bone2Name=Rig_Root, swing1LimitDegrees=45, swing2LimitDegrees=45, twistLimitDegrees=45}
 {bone1Name=Rig_Lip_Spring_006, bone2Name=Rig_Lip_002, swing1LimitDegrees=45, swing2LimitDegrees=45, twistLimitDegrees=45}
 {bone1Name=Rig_Lip_003, bone2Name=Rig_Root, swing1LimitDegrees=45, swing2LimitDegrees=45, twistLimitDegrees=45}
 {bone1Name=Rig_Lip_Spring_005, bone2Name=Rig_Lip_003, swing1LimitDegrees=45, swing2LimitDegrees=45, twistLimitDegrees=45}
 {bone1Name=Rig_Lip_001, bone2Name=Rig_Root, swing1LimitDegrees=45, swing2LimitDegrees=45, twistLimitDegrees=45}
 {bone1Name=MouthInner_006, bone2Name=Rig_Root, swing1LimitDegrees=45, swing2LimitDegrees=45, twistLimitDegrees=45}
 {bone1Name=MouthInner_002, bone2Name=Rig_Root, swing1LimitDegrees=45, swing2LimitDegrees=45, twistLimitDegrees=45}
 {bone1Name=MouthInner_003, bone2Name=Rig_Root, swing1LimitDegrees=45, swing2LimitDegrees=45, twistLimitDegrees=45}
 {bone1Name=MouthInner_001, bone2Name=Rig_Root, swing1LimitDegrees=45, swing2LimitDegrees=45, twistLimitDegrees=45}
 {bone1Name=MouthInner_004, bone2Name=Rig_Root, swing1LimitDegrees=45, swing2LimitDegrees=45, twistLimitDegrees=45}
 {bone1Name=MouthInner_008, bone2Name=Rig_Root, swing1LimitDegrees=45, swing2LimitDegrees=45, twistLimitDegrees=45}
 {bone1Name=MouthInner_007, bone2Name=Rig_Root, swing1LimitDegrees=45, swing2LimitDegrees=45, twistLimitDegrees=45}
 {bone1Name=MouthInner_005, bone2Name=Rig_Root, swing1LimitDegrees=45, swing2LimitDegrees=45, twistLimitDegrees=45}
 {bone1Name=Rig_Teeth_Uvula_001, bone2Name=Rig_Root, swing1LimitDegrees=45, swing2LimitDegrees=45, twistLimitDegrees=45}
 {bone1Name=Rig_Teeth_Uvula_003, bone2Name=Rig_Teeth_Uvula_001, swing1LimitDegrees=45, swing2LimitDegrees=45, twistLimitDegrees=45}
 {bone1Name=Rig_Eye_R, bone2Name=Rig_Root, swing1LimitDegrees=45, swing2LimitDegrees=45, twistLimitDegrees=45}
 {bone1Name=Rig_Eye_L_Parent, bone2Name=Rig_Root, swing1LimitDegrees=45, swing2LimitDegrees=45, twistLimitDegrees=45}
 {bone1Name=Rig_Teeth_Top, bone2Name=Rig_Root, swing1LimitDegrees=45, swing2LimitDegrees=45, twistLimitDegrees=45}
 {bone1Name=Rig_Teeth_Bottom, bone2Name=Rig_Root, swing1LimitDegrees=45, swing2LimitDegrees=45, twistLimitDegrees=45}
]
